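# Sound cue for the start of the ReactorGrade targeting beam, stored with the Tether item gameplay cues.
# A random node picks between two beam waves at equal weight, then a sound class node routes the result to a third-person ranged weapon fire class.
# It plays at reduced volume with whoosh-by attenuation, keeps playing when silent, and limits concurrency per owner by stopping the oldest.
# The debug flag is on.
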SoundCue /MistRaven/GameplayCues/Tether/Items/ReactorGrade/Audio/ReactorGrade_TargetingBeam_Start_Cue
root=SoundClass_5
nodes [
 {id=Random_0, children [WavePlayer_0,WavePlayer_2], properties {weights [1,1]}}
 {id=SoundClass_5, children [Random_0], properties {soundClassOverride=&Fort_WeaponRanged_Fire_3P_PSM}}
 {id=WavePlayer_0, properties {soundWaveAssetPtr=&/ReactorGrade_TargetingBeam_01}}
 {id=WavePlayer_2, properties {soundWaveAssetPtr=&/ReactorGrade_TargetingBeam_02}}
]
properties {
 attenuationSettings=&/ReactorGrade_Whoosh_By_Att
 bDebug=true
 bOverrideConcurrency=true
 maxDistance=2048
 virtualizationMode=PlayWhenSilent
 volumeMultiplier=0.4
 concurrencyOverrides {bLimitToOwner=true, maxCount=2, resolutionRule=StopOldest}
}
